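# Niagara landmark effect for WhisperRemedySour that plays a short swirling mesh burst, apparently soft airflow.
# NS_SoftAirflow spawns three meshes after a brief delay, using execution-index arrays to vary each one's scale and orientation.
# Curves shrink the vertically stretched meshes, accelerate their spin, fade alpha in and out, and drive dynamic material parameters.
# The Minimal sprite emitter is disabled, and no shown module reads User.BurstDelay.
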NiagaraSystem /WhisperRemedySour/Assets/FX/NS_WhisperRenedy_LandMark
dependencies {
 usedDataInterfaces [&NiagaraDataInterfaceArrayFloat, &NiagaraDataInterfaceArrayFloat3, &NiagaraDataInterfaceCurve]
 usedDynamicInputs []
 usedModules [
  &self:SystemSpawnScript.SystemScriptSource.SystemScriptGraph.NiagaraNodeAssignment_0.SetVariables_#1
  &SystemState
 ]
}
emitters {
 Minimal {
  bEnabled=false
  emitterUpdateScript [
   &EmitterState()
   &SpawnBurst_Instantaneous(Spawn Count=10000, Use Spawn Probability=false, UseLoopCountLimit=false)
  ]
  particleSpawnScript [
   &InitializeParticle(
    Color Mode=::NewEnumerator0
    Lifetime=5
    Sprite Size Mode=::NewEnumerator3
    Uniform Sprite Size=1000
    UsePositionOffset=false
    Write Material Random=true
   )
   &ShapeLocation(MaskSpawnGroup=false, Override Seed=false, Sphere Radius=49245.4414, Sphere Surface Distribution=1, UseOwnerScale=true)
  ]
  particleUpdateScript [&ParticleState()]
  rendererClasses [&NiagaraSpriteRendererProperties]
  renderers [{rendererClass=&NiagaraSpriteRendererProperties}]
 }
 NS_SoftAirflow {
  emitterUpdateScript [
   &EmitterState()
   &SpawnBurst_Instantaneous(Spawn Count=3, Spawn Time=0.3, Use Spawn Probability=false, UseLoopCountLimit=false)
  ]
  particleSpawnScript [
   &InitializeParticle(
    Lifetime Min=2
    Lifetime Mode=::NewEnumerator1
    Mesh Scale=&Multiply_VectorByFloat(
     Vector=<6 6 6>
     Float=&SelectFloatFromArray(
      Direct Array Index=&ReturnExecIndex()
      Float Selection Array=DataInterface() {propertyValues {FloatData [1.2,1,1.4]}}
     )
    )
    Mesh Scale Mode=::NewEnumerator4
    Position Offset=<0 0 200>
    UsePositionOffset=false
    Write Material Random=true
   )
   &InitialMeshOrientation(
    Mesh Orientation Mode=::NewEnumerator2
    Orientation Vector=<0 0 1>
    Rotation=&SelectVectorFromArray(
     Direct Array Index=&ReturnExecIndex()
     Direct Array Mode=::NewEnumerator1
     Vector Selection Array=DataInterface() {propertyValues {FloatData [<550.6341 104.9188 33.2165>, <60 -39.9795 1>, <0 0 0>]}}
    )
    Use Rotation Vector=true
   )
   &AddVelocity()
   &ApplyInitialForces()
  ]
  particleUpdateScript [
   &ParticleState()
   &DynamicMaterialParameters(
    Index 0 Param 1=&FloatFromCurve(FloatCurve=DataInterface() {propertyValues {Curve {keys [{interpMode=RCIM_Cubic, tangentMode=RCTM_User, time=0.5, value=0.2}, {interpMode=RCIM_Cubic, tangentMode=RCTM_User, time=0.9, value=0}]}}})
    Index 0 Param 2=1
    Index 0 Param 3=1
    Module.Param0WriteEnabled=true
    Module.Param1WriteEnabled=true
    Module.Param2WriteEnabled=true
    Module.Param3WriteEnabled=true
   )
   &ScaleMeshSize(Scale Factor=&Multiply_VectorByFloat(
     Vector=<1 1 1.5>
     Float=&FloatFromCurve(FloatCurve=DataInterface() {propertyValues {Curve {
         defaultValue=0.3
         keys [
          {arriveTangent=-10.1464, interpMode=RCIM_Cubic, leaveTangent=-10.1464, tangentMode=RCTM_User, time=0, value=1.8895}
          {arriveTangent=-1.2586, interpMode=RCIM_Cubic, leaveTangent=-1.2586, tangentMode=RCTM_User, time=0.2016, value=0.9979}
          <at 1 0.6 RCIM_Cubic>
         ]
        }}})
    ))
   &UpdateMeshOrientation(Rotation Rate=&FloatFromCurve(
     FloatCurve=DataInterface() {propertyValues {Curve {
        defaultValue=1
        keys [
         {interpMode=RCIM_Cubic, tangentMode=RCTM_User, time=0, value=0.3039}
         {arriveTangent=8.387, interpMode=RCIM_Cubic, leaveTangent=8.387, tangentMode=RCTM_User, time=0.7, value=2}
        ]
       }}}
     Scale Curve=&Multiply_Float(A=1.6281, B=&RandomRangeFloat(Minimum=1, Maximum=2))
    )) {module=UpdateMeshOrientation001}
   &ScaleColor(
    Scale Alpha=&FloatFromCurve(
     FloatCurve=DataInterface() {propertyValues {Curve {
        defaultValue=0
        keys [
         {arriveTangent=3.3238, interpMode=RCIM_Cubic, leaveTangent=3.3238, tangentMode=RCTM_User, time=0, value=0}
         {arriveTangent=1.2188, interpMode=RCIM_Cubic, leaveTangent=1.2188, time=0.0772, value=0.2522}
         {arriveTangent=-0.2462, interpMode=RCIM_Cubic, leaveTangent=-0.2462, time=0.8, value=0.975}
         {interpMode=RCIM_Cubic, leaveTangent=-0.0405, time=1, value=0.025}
        ]
       }}}
     Scale Curve=0.5
    )
    Scale RGB=&VectorFromFloat(Value=1)
    ScaleA=true
    ScaleRGB=true
   )
  ]
  rendererClasses [&NiagaraMeshRendererProperties]
  renderers [{rendererClass=&NiagaraMeshRendererProperties, rendererIndex=0}]
 }
}
summary {
 systemName=NS_WhisperRenedy_LandMark
 emitters [
  {emitterName=NS_SoftAirflow, rendererClasses [&NiagaraMeshRendererProperties]}
  {bEnabled=false, emitterName=Minimal, rendererClasses [&NiagaraSpriteRendererProperties]}
 ]
 userVariables [User.BurstDelay {type=NiagaraFloat, defaultValue=0.1, description=""}]
}
