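# Sound Cue for the looping aiming sound of WhisperRemedySour, routed through the Fortnite gadgets sound class.
# A single looping wave player feeds the output directly, reusing an Athena BiomeScanner quest ambient loop.
# It borrows the WhisperRemedySour Strike attenuation and slightly boosts volume.
# It primes on load and overrides concurrency to refuse new instances once the limit is reached.
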SoundCue /WhisperRemedySour/Assets/Sound/WhisperRemedySour_Aiming_Cue
root=WavePlayer_2
nodes [{id=WavePlayer_2, properties {bLooping=true, soundWaveAssetPtr=/Game/Athena/Sounds/Quests/BiomeScanner/BiomeScanner_AmbientLoop01}}]
properties {
 attenuationSettings=/WhisperRemedySour/Assets/Sound/Settings/WhisperRemedySour_Strike_Att
 bOverrideConcurrency=true
 bPrimeOnLoad=true
 maxDistance=9000
 soundClassObject=/Game/Sounds/FortSoundClassesAndModes/Fort_Gadgets_StW
 volumeMultiplier=1.2
 concurrencyOverrides {resolutionRule=PreventNew}
}
