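# Sound Cue for the looping ambient sound of the WhisperRemedySour loot item.
# A looping wave player runs through a fixed short Delay, an Enveloper with a brief volume fade-in, and a Modulator at default settings as root.
# The cue plays at reduced volume using shared item ambient-loop attenuation, the loot ambient sound class and loot ambient concurrency rules.
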SoundCue /WhisperRemedySour/Assets/Sound/WhisperRemedySour_Ambient_Loop_Cue
root=Modulator_0
nodes [
 {id=Delay_0, children [WavePlayer_0], properties {delayMax=0.25, delayMin=0.25}}
 {
  id=Enveloper_0
  children [Delay_0]
  properties {pitchCurve {editorCurveData {keys [{time=0}]}}, volumeCurve {editorCurveData {keys [<at 0 0>,<at 0.3 1>]}}}
 }
 {id=Modulator_0, children [Enveloper_0]}
 {id=WavePlayer_0, properties {bLooping=true, soundWaveAssetPtr=&/IP_WhisperRemedySour_Ambient_Loop}}
]
properties {
 attenuationSettings=/Game/Sounds/Attenuation/Fort_Item_Ambient_Loop_Mono_Default_Att
 maxDistance=1250
 soundClassObject=/Game/Sounds/FortSoundClassesAndModes/Fort_Loot_Ambient_Loop
 volumeMultiplier=0.5
 concurrencySet [/Game/Sounds/Concurrency_Rules/Loot_Ambient_Conc]
}
