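# Sound cue for the WhisperRemedySour bounce effect, playing bone bounce waves with grenade-bounce attenuation in the projectile effects sound class.
# A Modulator with fixed pitch feeds a Quality Level node that selects between two equally weighted Random nodes.
# The first two quality slots draw from all six variations, while the last two share a three-wave subset excluded from branch culling.
# A concurrency override caps simultaneous instances at three and rejects new ones.
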SoundCue /WhisperRemedySour/Assets/Sound/WhisperRemedySour_Bounce_Cue
root=Modulator_5
nodes [
 {id=Modulator_5, children [QualityLevel_0], properties {pitchMax=1, pitchMin=1}}
 {id=QualityLevel_0, children [Random_4,Random_4,Random_0,Random_0]}
 {id=Random_0, children [WavePlayer_0,WavePlayer_2,WavePlayer_5], properties {bShouldExcludeFromBranchCulling=true, weights [1,1,1]}}
 {id=Random_4, children [WavePlayer_0,WavePlayer_1,WavePlayer_2,WavePlayer_3,WavePlayer_4,WavePlayer_5], properties {weights [1,1,1,1,1,1]}}
 {id=WavePlayer_0, properties {soundWaveAssetPtr=&/WhisperRemedySour_BoneBounce_04}}
 {id=WavePlayer_1, properties {soundWaveAssetPtr=&/WhisperRemedySour_BoneBounce_03}}
 {id=WavePlayer_2, properties {soundWaveAssetPtr=&/WhisperRemedySour_BoneBounce_02}}
 {id=WavePlayer_3, properties {soundWaveAssetPtr=&/WhisperRemedySour_BoneBounce_01}}
 {id=WavePlayer_4, properties {soundWaveAssetPtr=&/WhisperRemedySour_BoneBounce_05}}
 {id=WavePlayer_5, properties {soundWaveAssetPtr=&/WhisperRemedySour_BoneBounce_06}}
]
properties {
 attenuationSettings=/Game/Sounds/Attenuation/Fort_Grenade_Bounce_Att
 bOverrideConcurrency=true
 maxDistance=5500
 soundClassObject=&Fort_WeaponRanged_Projectile_Effects
 volumeMultiplier=1.5
 concurrencyOverrides {maxCount=3, resolutionRule=PreventNew}
}
