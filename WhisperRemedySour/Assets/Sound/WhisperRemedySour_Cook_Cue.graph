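# Sound cue for the cooking sound of the WhisperRemedySour item, classed under gadget and item sounds.
# A random node picks one of two equally weighted cook waves, and a modulator raises its pitch within a narrow random range at fixed volume.
# It plays quietly with weapon-reload attenuation and overrides concurrency to prevent new instances once the limit is reached.
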SoundCue /WhisperRemedySour/Assets/Sound/WhisperRemedySour_Cook_Cue
root=Modulator_0
nodes [
 {id=Modulator_0, children [Random_0], properties {pitchMax=1.3, pitchMin=1.2, volumeMax=1, volumeMin=1}}
 {id=Random_0, children [WavePlayer_0,WavePlayer_2], properties {weights [1,1]}}
 {id=WavePlayer_0, properties {soundWaveAssetPtr=&/WhisperRemedySour_Cook_01}}
 {id=WavePlayer_2, properties {soundWaveAssetPtr=&/WhisperRemedySour_Cook_02}}
]
properties {
 attenuationSettings=&Fort_Weapon_Reload_Att
 bOverrideConcurrency=true
 maxDistance=2250
 soundClassObject=&Fort_Gadgets_Items
 volumeMultiplier=0.2
 concurrencyOverrides {resolutionRule=PreventNew}
}
