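# Sound cue for dropping the WhisperRemedySour item, playing a whoosh at reduced volume.
# A Random node picks one of three equally weighted DropWhoosh waves and feeds it through a Modulator at the root.
# It uses weapon pickup attenuation and the loot collect sound class.
# A concurrency override allows only one instance, blocking new plays while one is active.
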SoundCue /WhisperRemedySour/Assets/Sound/WhisperRemedySour_Drop_Cue
root=Modulator_0
nodes [
 {id=Modulator_0, children [Random_0]}
 {id=Random_0, children [WavePlayer_0,WavePlayer_1,WavePlayer_2], properties {weights [1,1,1]}}
 {id=WavePlayer_0, properties {soundWaveAssetPtr=&/IP_WhisperRemedySour_DropWhoosh_01}}
 {id=WavePlayer_1, properties {soundWaveAssetPtr=&/IP_WhisperRemedySour_DropWhoosh_02}}
 {id=WavePlayer_2, properties {soundWaveAssetPtr=&/IP_WhisperRemedySour_DropWhoosh_03}}
]
properties {
 attenuationSettings=&Fort_Weapon_Pickup_Att
 bOverrideConcurrency=true
 maxDistance=3500
 soundClassObject=&Fort_Loot_Collect
 volumeMultiplier=0.5
 concurrencyOverrides {maxCount=1, resolutionRule=PreventNew}
}
