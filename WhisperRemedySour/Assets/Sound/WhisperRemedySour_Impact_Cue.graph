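# Sound cue for the WhisperRemedySour impact, mixing close and distant layers, each with its own attenuation, under the large-explosives sound class.
# Each layer runs through a Modulator into a Quality Level switch that picks a random wave from either the full variant set or a two-wave subset.
# The reduced subsets are excluded from branch culling.
# The cue primes on load and overrides concurrency to cap simultaneous instances.
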SoundCue /WhisperRemedySour/Assets/Sound/WhisperRemedySour_Impact_Cue
root=Mixer_0
nodes [
 {
  id=Attenuation_0
  children [Modulator_0]
  properties {attenuationSettings=/WhisperRemedySour/Assets/Sound/Settings/WhisperRemedySour_Strike_Impact_Close_Att}
 }
 {
  id=Attenuation_1
  children [Modulator_1]
  properties {attenuationSettings=/WhisperRemedySour/Assets/Sound/Settings/WhisperRemedySour_Strike_Impact_Distant_Att}
 }
 {id=Mixer_0, children [Attenuation_0, Attenuation_1], properties {inputVolume [1,1]}}
 {id=Modulator_0, children [QualityLevel_0]}
 {id=Modulator_1, children [QualityLevel_1]}
 {id=QualityLevel_0, children [Random_0,Random_0,Random_1,Random_1]}
 {id=QualityLevel_1, children [Random_2,Random_2,Random_3,Random_3]}
 {id=Random_0, children [WavePlayer_1,WavePlayer_3,WavePlayer_4,WavePlayer_5], properties {weights [1,1,1,1]}}
 {id=Random_1, children [WavePlayer_3,WavePlayer_5], properties {bShouldExcludeFromBranchCulling=true, weights [1,1]}}
 {id=Random_2, children [WavePlayer_6, WavePlayer_7, WavePlayer_8, WavePlayer_9, WavePlayer_10], properties {weights [1,1,1,1,1]}}
 {id=Random_3, children [WavePlayer_8, WavePlayer_10], properties {bShouldExcludeFromBranchCulling=true, weights [1,1]}}
 {id=WavePlayer_1, properties {soundWaveAssetPtr=&/IP_WhisperRemedySour_Impact_01}}
 {id=WavePlayer_10, properties {soundWaveAssetPtr=&/IP_WhisperRemedySour_Impact_Distant_05}}
 {id=WavePlayer_3, properties {soundWaveAssetPtr=&/IP_WhisperRemedySour_Impact_03}}
 {id=WavePlayer_4, properties {soundWaveAssetPtr=&/IP_WhisperRemedySour_Impact_04}}
 {id=WavePlayer_5, properties {soundWaveAssetPtr=&/IP_WhisperRemedySour_Impact_05}}
 {id=WavePlayer_6, properties {soundWaveAssetPtr=&/IP_WhisperRemedySour_Impact_Distant_01}}
 {id=WavePlayer_7, properties {soundWaveAssetPtr=&/IP_WhisperRemedySour_Impact_Distant_02}}
 {id=WavePlayer_8, properties {soundWaveAssetPtr=&/IP_WhisperRemedySour_Impact_Distant_03}}
 {id=WavePlayer_9, properties {soundWaveAssetPtr=&/IP_WhisperRemedySour_Impact_Distant_04}}
]
properties {
 attenuationSettings=/WhisperRemedySour/Assets/Sound/Settings/WhisperRemedySour_Strike_Att
 bOverrideConcurrency=true
 bPrimeOnLoad=true
 maxDistance=10500
 soundClassObject=/Game/Sounds/FortSoundClassesAndModes/Fort_Explosives_Large_PSM
 volumeMultiplier=0.3
 concurrencyOverrides {maxCount=8}
}
