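# Sound cue for the WhisperRemedySour projectile's in-flight sound, in the ranged-weapon projectile effects class with its own attenuation.
# A mixer layers a reused Coal in-air loop at full level over a very quiet wind loop and half-level bone shine, spin and throw sweetener layers.
# Coal, wind and shine loop, while spin and sweetener are one-shots.
# Modulators pitch all but the Coal layer up by a fixed amount.
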SoundCue /WhisperRemedySour/Assets/Sound/WhisperRemedySour_InAir_Loop_Cue
root=Mixer_0
nodes [
 {id=Mixer_0, children [Modulator_0,Modulator_1,Modulator_2,Modulator_3,Modulator_4], properties {inputVolume [1,0.05,0.5,0.5,0.5]}}
 {id=Modulator_0, children [WavePlayer_1], properties {volumeMax=1, volumeMin=1}}
 {id=Modulator_1, children [WavePlayer_2], properties {pitchMax=1.5, pitchMin=1.5, volumeMax=1, volumeMin=1}}
 {id=Modulator_2, children [WavePlayer_3], properties {pitchMax=1.5, pitchMin=1.5, volumeMax=1, volumeMin=1}}
 {id=Modulator_3, children [WavePlayer_0], properties {pitchMax=1.5, pitchMin=1.5, volumeMax=1, volumeMin=1}}
 {id=Modulator_4, children [WavePlayer_4], properties {pitchMax=1.5, pitchMin=1.5, volumeMax=0.5, volumeMin=1}}
 {id=WavePlayer_0, properties {bLooping=true, soundWaveAssetPtr=&/IP_WhisperRemedySour_BoneShine_Loop}}
 {id=WavePlayer_1, properties {bLooping=true, soundWaveAssetPtr=/Game/Athena/Sounds/Items/Coal/Weapon_Item_Coal_InAir_Loop}}
 {id=WavePlayer_2, properties {bLooping=true, soundWaveAssetPtr=&/WhisperRemedySour_InAir_Wind_Loop}}
 {id=WavePlayer_3, properties {soundWaveAssetPtr=&/IP_WhisperRemedySour_BoneSpin}}
 {id=WavePlayer_4, properties {soundWaveAssetPtr=&/IP_WhisperRemedySour_BoneThrow_Sweetener}}
]
properties {
 attenuationSettings=/WhisperRemedySour/Assets/Sound/Settings/WhisperRemedySour_InAirLoop_Att
 maxDistance=6000
 soundClassObject=&Fort_WeaponRanged_Projectile_Effects
 volumeMultiplier=4
}
