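# Sound cue for the WhisperRemedySour incoming sound, mixing an incoming one-shot with an animal howl layer.
# The incoming layer uses a quality-level switch between a five-variant random pool and a three-variant subset, then a modulator and the Strike attenuation.
# The howl layer randomly picks one of three howls, delays it briefly, then applies its own modulator and animal attenuation.
# It primes on load and overrides concurrency to cap simultaneous instances.
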SoundCue /WhisperRemedySour/Assets/Sound/WhisperRemedySour_Incoming_Cue
root=Mixer_0
nodes [
 {
  id=Attenuation_2
  children [Modulator_0]
  properties {attenuationSettings=/WhisperRemedySour/Assets/Sound/Settings/WhisperRemedySour_Strike_Att}
 }
 {
  id=Attenuation_3
  children [Modulator_1]
  properties {attenuationSettings=/WhisperRemedySour/Assets/Sound/Settings/IP_WhisperRemedySour_Animal_Incommin_Att}
 }
 {id=Delay_0, children [Random_1], properties {delayMax=0.2, delayMin=0.2}}
 {id=Mixer_0, children [Attenuation_2, Attenuation_3], properties {inputVolume [1,1]}}
 {id=Modulator_0, children [QualityLevel_1]}
 {id=Modulator_1, children [Delay_0]}
 {id=QualityLevel_1, children [Random_0,Random_0,Random_3,Random_3]}
 {id=Random_0, children [WavePlayer_0,WavePlayer_1,WavePlayer_2,WavePlayer_3,WavePlayer_9], properties {weights [1,1,1,1,1]}}
 {id=Random_1, children [WavePlayer_4,WavePlayer_5,WavePlayer_6], properties {weights [1,1,1]}}
 {id=Random_3, children [WavePlayer_0,WavePlayer_2,WavePlayer_9], properties {bShouldExcludeFromBranchCulling=true, weights [1,1,1]}}
 {id=WavePlayer_0, properties {soundWaveAssetPtr=&/IP_WhisperRemedySour_Incoming_01}}
 {id=WavePlayer_1, properties {soundWaveAssetPtr=&/IP_WhisperRemedySour_Incoming_02}}
 {id=WavePlayer_2, properties {soundWaveAssetPtr=&/IP_WhisperRemedySour_Incoming_03}}
 {id=WavePlayer_3, properties {soundWaveAssetPtr=&/IP_WhisperRemedySour_Incoming_04}}
 {id=WavePlayer_4, properties {soundWaveAssetPtr=&/IP_WhisperRemedySour_Incoming_Howl_01}}
 {id=WavePlayer_5, properties {soundWaveAssetPtr=&/IP_WhisperRemedySour_Incoming_Howl_02}}
 {id=WavePlayer_6, properties {soundWaveAssetPtr=&/IP_WhisperRemedySour_Incoming_Howl_03}}
 {id=WavePlayer_9, properties {soundWaveAssetPtr=&/IP_WhisperRemedySour_Incoming_05}}
]
properties {
 attenuationSettings=&WhisperRemedySour_Strike_Att
 bOverrideConcurrency=true
 bPrimeOnLoad=true
 maxDistance=9000
 soundClassObject=&Fort_WeaponRanged_Fire_3P_PSM
 volumeMultiplier=1
 concurrencyOverrides {maxCount=8}
}
