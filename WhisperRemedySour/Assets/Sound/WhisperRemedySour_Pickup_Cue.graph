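# Sound cue for picking up the WhisperRemedySour item, layering a main pickup sound, a bone bounce and a sweetener.
# A mixer plays the main wave immediately at full level, then after a short fixed delay adds one of three random bone bounce variants and the sweetener, both quieter and pitched up.
# It uses loot-collect class and weapon pickup attenuation, with concurrency limited to one instance that blocks new plays.
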SoundCue /WhisperRemedySour/Assets/Sound/WhisperRemedySour_Pickup_Cue
root=Mixer_0
nodes [
 {id=Delay_0, children [Random_4], properties {delayMax=0.3, delayMin=0.3}}
 {id=Delay_1, children [WavePlayer_7], properties {delayMax=0.3, delayMin=0.3}}
 {id=Mixer_0, children [Modulator_0,Modulator_1,Modulator_2], properties {inputVolume [1,0.7,0.5]}}
 {id=Modulator_0, children [WavePlayer_0], properties {pitchMax=1, pitchMin=1}}
 {id=Modulator_1, children [Delay_0], properties {pitchMax=1.2, pitchMin=1.2}}
 {id=Modulator_2, children [Delay_1], properties {pitchMax=1.2, pitchMin=1.2}}
 {id=Random_4, children [WavePlayer_1,WavePlayer_2,WavePlayer_3], properties {weights [1,1,1]}}
 {id=WavePlayer_0, properties {soundWaveAssetPtr=&/IP_WhisperRemedySour_PickUp}}
 {id=WavePlayer_1, properties {soundWaveAssetPtr=&/WhisperRemedySour_BoneBounce_04}}
 {id=WavePlayer_2, properties {soundWaveAssetPtr=&/WhisperRemedySour_BoneBounce_03}}
 {id=WavePlayer_3, properties {soundWaveAssetPtr=&/WhisperRemedySour_BoneBounce_02}}
 {id=WavePlayer_7, properties {soundWaveAssetPtr=&/IP_WhisperRemedySour_Pickup_Sweetener}}
]
properties {
 attenuationSettings=&Fort_Weapon_Pickup_Att
 bOverrideConcurrency=true
 maxDistance=3500
 soundClassObject=&Fort_Loot_Collect
 volumeMultiplier=1
 concurrencyOverrides {maxCount=1, resolutionRule=PreventNew}
}
