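# WhisperRemedySour placement sound cue, going by its name, that plays a single Ashton Milo launcher charge-start wave.
# The wave player is the root, with no randomization or mixing nodes.
# It uses the WhisperRemedySour strike attenuation and the gadgets sound class at slightly reduced volume, and primes on load.
# A concurrency override prevents new instances from starting once the limit is reached.
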SoundCue /WhisperRemedySour/Assets/Sound/WhisperRemedySour_Placement_Cue
root=WavePlayer_3
nodes [{id=WavePlayer_3, properties {soundWaveAssetPtr=/Game/Athena/Sounds/LTMs/Ashton/Milo/Ashton_Milo_Launcher_Charge_Start_03}}]
properties {
 attenuationSettings=/WhisperRemedySour/Assets/Sound/Settings/WhisperRemedySour_Strike_Att
 bOverrideConcurrency=true
 bPrimeOnLoad=true
 maxDistance=9000
 soundClassObject=/Game/Sounds/FortSoundClassesAndModes/Fort_Gadgets_StW
 volumeMultiplier=0.8
 concurrencyOverrides {resolutionRule=PreventNew}
}
