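# Sound cue for the WhisperRemedySour take-off, layering a randomized exit sound with a randomized boom.
# A mixer feeds two modulated branches: one picks among three boom waves, the other routes exit waves through a quality-level switch.
# The first two quality slots draw from four exit variations, the last two from a two-wave subset.
# It uses the Strike attenuation, a third-person ranged-weapon fire sound class, prime-on-load and a concurrency override.
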SoundCue /WhisperRemedySour/Assets/Sound/WhisperRemedySour_TakeOff_Cue
root=Mixer_0
nodes [
 {id=Mixer_0, children [Modulator_1,Modulator_0], properties {inputVolume [1,1]}}
 {id=Modulator_0, children [Random_1]}
 {id=Modulator_1, children [QualityLevel_1]}
 {id=QualityLevel_1, children [Random_0,Random_0,Random_3,Random_3]}
 {id=Random_0, children [WavePlayer_1,WavePlayer_2,WavePlayer_3,WavePlayer_4], properties {weights [1,1,1,1]}}
 {id=Random_1, children [WavePlayer_0,WavePlayer_5,WavePlayer_6], properties {weights [1,1,1]}}
 {id=Random_3, children [WavePlayer_2,WavePlayer_4], properties {bShouldExcludeFromBranchCulling=true, weights [1,1]}}
 {id=WavePlayer_0, properties {soundWaveAssetPtr=&/IP_WhisperRemedySour_Exit_Boom_01}}
 {id=WavePlayer_1, properties {soundWaveAssetPtr=&/IP_WhisperRemedySour_Exit_01}}
 {id=WavePlayer_2, properties {soundWaveAssetPtr=&/IP_WhisperRemedySour_Exit_02}}
 {id=WavePlayer_3, properties {soundWaveAssetPtr=&/IP_WhisperRemedySour_Exit_03}}
 {id=WavePlayer_4, properties {soundWaveAssetPtr=&/IP_WhisperRemedySour_Exit_04}}
 {id=WavePlayer_5, properties {soundWaveAssetPtr=&/IP_WhisperRemedySour_Exit_Boom_02}}
 {id=WavePlayer_6, properties {soundWaveAssetPtr=&/IP_WhisperRemedySour_Exit_Boom_03}}
]
properties {
 attenuationSettings=/WhisperRemedySour/Assets/Sound/Settings/WhisperRemedySour_Strike_Att
 bOverrideConcurrency=true
 bPrimeOnLoad=true
 maxDistance=9000
 soundClassObject=&Fort_WeaponRanged_Fire_3P_PSM
 volumeMultiplier=1
 concurrencyOverrides {maxCount=8}
}
