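# Sound cue for throwing the WhisperRemedySour gadget, mixing a throw sound with a delayed whistle.
# The throw branch picks one of three equally weighted waves, varies its pitch and uses the shared gadget-throw attenuation.
# The whistle branch starts after a fixed half-second delay, with boosted volume, slight pitch variation and its own attenuation, and is weighted much louder in the mix.
# The cue uses the gadget items sound class and overrides concurrency to block new instances at the limit.
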SoundCue /WhisperRemedySour/Assets/Sound/WhisperRemedySour_Throw_Cue
root=Mixer_2
nodes [
 {
  id=Attenuation_0
  children [Delay_1]
  properties {attenuationSettings=/WhisperRemedySour/Assets/Sound/Settings/WhisperRemedySour_Whistle_Att}
 }
 {id=Attenuation_1, children [Modulator_1], properties {attenuationSettings=/Game/Sounds/Attenuation/Fort_GadgetThrow_Att}}
 {id=Delay_1, children [Modulator_0], properties {delayMax=0.5, delayMin=0.5}}
 {id=Mixer_2, children [Attenuation_1, Attenuation_0], properties {inputVolume [0.8,3]}}
 {id=Modulator_0, children [WavePlayer_5], properties {pitchMax=1.1, pitchMin=1, volumeMax=2, volumeMin=2}}
 {id=Modulator_1, children [Random_1], properties {pitchMax=1.2, pitchMin=1}}
 {id=Random_1, children [WavePlayer_6,WavePlayer_7,WavePlayer_8], properties {weights [1,1,1]}}
 {id=WavePlayer_5, properties {soundWaveAssetPtr=&/IP_WhisperRemedySour_Whistle}}
 {id=WavePlayer_6, properties {soundWaveAssetPtr=&/WhisperRemedySour_Throw01}}
 {id=WavePlayer_7, properties {soundWaveAssetPtr=&/WhisperRemedySour_Throw02}}
 {id=WavePlayer_8, properties {soundWaveAssetPtr=&/WhisperRemedySour_Throw03}}
]
properties {
 bOverrideConcurrency=true
 maxDistance=3500
 soundClassObject=&Fort_Gadgets_Items
 volumeMultiplier=0.5
 concurrencyOverrides {resolutionRule=PreventNew}
}
